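; Laser grid building actor that marks players who trigger it and alerts nearby guards, with editor helpers for placing outlets.
; On authority, an async gameplay tag query toggles the grid, and triggers multicast beam FX, fire cues and apply mark effects.
; State changes drive audio and outlet material custom data, or register an AI danger zone where cosmetics don't run.
; Deprecated functions document nativized per-instance HISM animation logic.
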(event EventBeginPlay
  ;[0.2896 0.875 0.2871] BeginPlay: Initialize the buildling actor.
  ; For some reason, the OnBuildingActorInitialized isn't called when this actor is spawned through a level instance.
  (|CallParentFunction)
  (Building|InitializeKismetSpawnedBuildingActor 0)
  ;;[0.9805 0.1936 1] Destroy the laser grid if it is hotfixed off.
  (Class|FortAsyncActionCurrentPlaylistReady|PlaylistDataReadyAsync
    (:OnReady
      (if (not (ScalableFloat|GetValueasBool (Variables|Hotfix|GetHotfix_LaserGameplayEnabled)))
        (Gameplay|SilentDie false)
        (Networking|FlushNetDormancy)))))

(event Building|EventOnBuildingActorInitialized
  (params (InitializationReason :type EFortBuildingInitializationReason%20Enum :default "StaticallyPlaced") (BuildingPersistentState :type EFortBuildingPersistentState%20Enum :default "Default"))
  (bind _activetagquery (Variables|Settings|GetActiveTagQuery))
  ;[0.02423 0.1823 0.04704] OnBuildingActorInitialized: Setup async latent actions that fire when the laser is supposed to be active/disabled based on a gameplay tag query.
  (|CallParentFunction InitializationReason BuildingPersistentState)
  (Networking|SwitchHasAuthority
    (:Authority
      (Class|FortAsyncActionCurrentPlaylistReady|PlaylistDataReadyAsync
        (:OnReady
          (bind _asynctaskproxy (Ability|Tasks|WaitGameplayTagQueryOnActor 0 _activetagquery)
            (:then
              (Variables|Transient|SetAsyncAction_ActiveTagQueryTrue _asynctaskproxy)
              (bind _asynctaskproxy_2 (Ability|Tasks|WaitGameplayTagQueryOnActor 0 _activetagquery "WhenFalse")
                (:then
                  (Variables|Transient|SetAsyncAction_ActiveTagQueryFalse _asynctaskproxy_2))
                (:Triggered
                  (Laser|SetLaserGridEnable false))))
            (:Triggered
              (Laser|SetLaserGridEnable true))))))))

(event Laser|EventCosmeticDestroyOutletInstance
  (params (OutletEntry :type Laser%20Grid%20Outlet%20Entry%20Structure%20%28by%20ref%29))
  (bind _instanceindex (Utilities|Struct|BreakLaserGridOutletEntry OutletEntry))
  (bind _returnvalue (Math|Conversions|ToInteger(Byte) _instanceindex))
  ;[0.209 0.1957 0.2135] CosmeticDestroyOutletInstance: Adjust the material of the outlet so that it is invisible.
  (bind _hismcomponentoutlet (Variables|Laser|GetHISMComponentOutlet)
    (:then
      (if (Components|InstancedStaticMesh|IsValidInstance _hismcomponentoutlet _returnvalue)
        (bind _outinstancetransform (Components|InstancedStaticMesh|GetInstanceTransform _hismcomponentoutlet _returnvalue false))
        (bind (_outinstancetransform_location _outinstancetransform_rotation) (Math|Transform|BreakTransform _outinstancetransform))
        (Components|InstancedStaticMesh|UpdateInstanceTransform _hismcomponentoutlet _returnvalue (Math|Transform|MakeTransform _outinstancetransform_location _outinstancetransform_rotation "0.000000,0.000000,0.000000") false true false)
        (bind _outinstancetransform_2 (Components|InstancedStaticMesh|GetInstanceTransform _hismcomponentoutlet _returnvalue true))
        (bind (_outinstancetransform_location_2 _outinstancetransform_rotation_2) (Math|Transform|BreakTransform _outinstancetransform_2))
        (Niagara|SpawnSystematLocation (Variables|FX|GetOutletDestroyFX) _outinstancetransform_location_2 _outinstancetransform_rotation_2 "1.000000,1.000000,1.000000" true true "AutoRelease")))
    (:else)))

(event Laser|EventLaserGridTriggered
  (params (LaserConnectionData :type Laser%20Grid%20Connection%20Entry%20Structure%20%28by%20ref%29) (TriggeringActor :type Actor%20Object%20Reference))
  ;[0.09633 0.02398 0.1406] LaserGridTriggered: When the laser grid is triggered, execute the laser triggered gameplay cue and mark the player
  (Networking|SwitchHasAuthority
    (:Authority
      (CallFunction|ExecuteLaserGridTriggeredFX)
      ;;[0.1406 0.01605 0.1174] Execute gameplay cue on the player(s) that trigger the laser and mark them.
      (bind _playerpawns (CallFunction|GetPlayerPawnsfromTriggeringActor TriggeringActor))
      (for _array_element _playerpawns
        (Utilities|IsValid _array_element
          (:Is%20Valid
            (Utilities|FlowControl|Sequence
              (:then_0
                (if (CallFunction|ShouldPlayTriggeredCue _array_element)
                  (GameplayCue|ExecuteGameplayCueOnActor(Burst) _array_element (Variables|GameplayCue|GetPlayerTriggeredGameplayCueTag) (Ability|GameplayCue|MakeGameplayCueParameters 0.0))))
              (:then_1
                (if (ScalableFloat|GetValueasBool (Variables|Hotfix|GetHotfix_LaserPlayerMarkEnabled))
                  (CallFunction|MarkPlayerPawn _array_element))))))))))

(event Laser|EventLaserGridStateUpdated
  (params (bLaserGridEnabled :type Boolean))
  (bind _hotfix_laseraidangerzoneradius (Variables|Hotfix|GetHotfix_LaserAIDangerZoneRadius))
  (bind _returnvalue (ScalableFloat|GetValueatLevel _hotfix_laseraidangerzoneradius))
  (bind _returnvalue_2 (Class|AthenaAIServiceZoneManager|GetAthenaAIServiceZoneManager true))
  (bind _self self)
  ;[0.035 0.1071 0.1771] LaserGridStateUpdated: Play or fade out the audio component based if the laser grid is enabled or not.
  (Networking|CanExecuteCosmeticEvents
    (:True
      (bind _audio (Variables|Default|GetAudio)
        (:then
          (if bLaserGridEnabled
            (Audio|Components|Audio|FadeIn _audio 0.1 1.0 0.0 "Linear" :!class /Script/Engine.AudioComponent)
            (else
              ;;[0.6562 0.1038 0.2157] Fades out Transition Sound
              (Audio|PlaySoundatLocation "/LaserGameplay/Sounds/Laser_Disabled.Laser_Disabled" (Transformation|GetActorLocation))
              (Audio|Components|Audio|FadeOut _audio 0.2 0.0 "Linear" :!class /Script/Engine.AudioComponent))))
        (:else
          _))
      (bind _hismcomponentoutlet (Variables|Laser|GetHISMComponentOutlet)
        (:then
          (for _index (range (Components|InstancedStaticMesh|GetInstanceCount _hismcomponentoutlet))
            (if (Components|InstancedStaticMesh|IsValidInstance _hismcomponentoutlet _index)
              (Components|InstancedStaticMesh|SetCustomDataValue _hismcomponentoutlet _index (Variables|Laser|Material|GetOutletMaterialCustomDataIndex_DisableEmissivePulse) (Math|Conversions|ToFloat(Boolean) (not bLaserGridEnabled)) true))))
        (:else)))
    (:False
      ;;[0.1406 0.02147 0.1138] Enable or disable the laser grid area as a danger zone for AIs.
      (if (> _returnvalue 0)
        (Utilities|IsValid _returnvalue_2
          (:Is%20Valid
            (if bLaserGridEnabled
              (Class|AthenaAIServiceZoneManager|RegisterDangerZone _returnvalue_2 _self _returnvalue)
              (else
                (Class|AthenaAIServiceZoneManager|UnregisterDangerZone _returnvalue_2 _self)))))))))

(event Custom|MulticastLaserGridTriggeredFX
  (params (VFXTransforms :type Array%20of%20Transforms))
  ;[0.03726 0.1742 0.1979] MulticastLaserGridTriggeredFX: Spawn the laser triggered FXs to communicate that the laser grid has been triggered.
  (Networking|CanExecuteCosmeticEvents
    (:True
      (bind _lasertriggeredfx (Variables|FX|GetLaserTriggeredFX)
        (:then
          (bind _object (Utilities|AsyncLoadAsset _lasertriggeredfx)
            (:Completed
              (bind _asniagara_system (Utilities|Casting|CastToNiagaraSystem _object))
              (for _array_element VFXTransforms
                (bind (_location _rotation _scale) (.location _array_element))
                (bind _scale_x (Math|Vector|BreakVector _scale))
                (bind _returnvalue (Niagara|SpawnSystematLocation _asniagara_system _location _rotation "1.000000,1.000000,1.000000" true true "AutoRelease"))
                (Utilities|IsValid _returnvalue
                  (:Is%20Valid
                    (Niagara|SetNiagaraVariable(Float) _returnvalue "ScaleBeam" _scale_x)))))))
        (:else)))))

(event Laser|EventCosmeticLaserConnectionsUpdated
  (params (NewLaserGridConnections :type Array%20of%20Laser%20Grid%20Connection%20Entry%20Structures))
  (CallFunction|UpdateLaserConnectionCosmetics NewLaserGridConnections))

(event EventEndPlay
  (params (EndPlayReason :type EEndPlayReason%20Enum :default "Destroyed"))
  (bind _returnvalue (Class|AthenaAIServiceZoneManager|GetAthenaAIServiceZoneManager true))
  ;[0.09375 0.05682 0.01248] EndPlay: End the async actions.
  (Utilities|FlowControl|Sequence
    (:then_0
      (Networking|SwitchHasAuthority
        (:Authority
          (Ability|Async|EndAction (Variables|Transient|GetAsyncAction_ActiveTagQueryFalse))
          (Utilities|IsValid _returnvalue
            (:Is%20Valid
              (Class|AthenaAIServiceZoneManager|UnregisterDangerZone _returnvalue self))))))
    (:then_1
      (|CallParentFunction EndPlayReason))))

(fn AddInstance
  ;[0.09297 0.02845 0.125] Add an instance to the outlet mesh component.
  (if (Editor|InEditor)
    (bind _hismcomponentoutlet (Variables|Laser|GetHISMComponentOutlet)
      (:then
        (Components|InstancedStaticMesh|AddInstance _hismcomponentoutlet (struct :Location "0, 0, 0" :Rotation "0, 0, 0" :Scale "0.500000,0.500000,0.500000") false))
      (:else))))

(fn CanActorTriggerLaserGrid
  (params (TriggeringActor :type Actor%20Object%20Reference))
  ;[0.0501 0.1875 0.1744] Returns if the actor can trigger the laser grid.
  (Utilities|FlowControl|Sequence
    (:then_0
      (bind _asfort_athena_aibot_controller (Utilities|Casting|CastToFortAthenaAIBotController (Fortnite|GetControllerfromActor TriggeringActor)))
      (if (not (Class|FortPlayerStateAthena|ControllerIsAnAthenaGameParticipant _asfort_athena_aibot_controller))
        (if (ScalableFloat|GetValueasBool (Variables|Hotfix|GetHotfix_LaserHiredNpcMarkEnabled))
          (bind _returnvalue (GameplayTags|HasMatchingGameplayTag(Message) (Variables|GetTriggeringActor) (Variables|Laser|GetHiredNpcTag))) ; Can mark non participant if we have the hired tag
          (return _returnvalue)
          (else
            (return false)))))
    (:then_1
      (bind _returnvalue_2 (|CallParentFunction TriggeringActor))
      (return _returnvalue_2))))

(fn DefaultOutlets
  ;[0.03264 0.09791 0.2292] DefaultOutlets: Resets the outlet instances to their default values.
  (if (Editor|InEditor)
    (bind _hismcomponentoutlet (Variables|Laser|GetHISMComponentOutlet)
      (:then
        (Components|InstancedStaticMesh|ClearInstances _hismcomponentoutlet)
        (Components|InstancedStaticMesh|AddInstances _hismcomponentoutlet (Variables|Settings|GetDefaultOutletTransforms) false false true))
      (:else))))

(fn DeprecatedCosmeticUpdateOutletInstance
  (params (OutletEntry :type Laser%20Grid%20Outlet%20Entry%20Structure%20%28by%20ref%29))
  (bind (_instanceindex _state) (Utilities|Struct|BreakLaserGridOutletEntry OutletEntry))
  (bind (_x _y) (Math|Vector2D|BreakVector2D (Variables|Laser|Material|GetOutletMaterialTransitionOffTimestampOffset)))
  (bind _outinstancetransform (Components|InstancedStaticMesh|GetInstanceTransform (Variables|Laser|GetHISMComponentOutlet) (LocalVariables|GetInstanceIndex) true))
  (bind _outinstancetransform_location (Math|Transform|BreakTransform _outinstancetransform))
  (bind _instanceindex_2 (LocalVariables|GetInstanceIndex))
  (bind _instanceindex_3 (LocalVariables|GetInstanceIndex))
  (bind _hismcomponentoutlet (Variables|Laser|GetHISMComponentOutlet))
  (bind _hismcomponentoutlet_2 (Variables|Laser|GetHISMComponentOutlet))
  (bind _instanceindex_4 (LocalVariables|GetInstanceIndex))
  (bind _hismcomponentoutlet_3 (Variables|Laser|GetHISMComponentOutlet))
  ;[0 0 0] CosmeticUpdateOutletInstance: Update the custom data value on the instance material to reflect the latest outlet state.
  ; THIS FUNCTION HAS BEEN NATIVIZED AND IS NOW DEPRECATED. IT IS ONLY HERE TO HELP ARTISTS AND DEV COMMUNICATE REGARDING THE EXISTING LOGIC.
  (Networking|CanExecuteCosmeticEvents
    (:True
      (bind _output_get (LocalVariables|SetInstanceIndex (Math|Conversions|ToInteger(Byte) _instanceindex)))
      (bind _hismcomponentoutlet_4 (Variables|Laser|GetHISMComponentOutlet)
        (:then
          (if (Components|InstancedStaticMesh|IsValidInstance _hismcomponentoutlet_4 _output_get)
            (switch Utilities|FlowControl|Switch|SwitchonELaserConnectionState _state
              (:Off
                ;;[0.04405 0.1519 0.1979] Outlet is OFF. Emissive is disabled.
                (Components|InstancedStaticMesh|SetCustomDataValue (Variables|Laser|GetHISMComponentOutlet) (LocalVariables|GetInstanceIndex) (Variables|Laser|Material|GetOutletMaterialCustomDataIndex_EmissiveEnabled) 0.0 true))
              (:TransitionOn
                ;;[0.02661 0.07292 0.03511] Outlet is transitioning on - enable the emissive and increase the intensity over time for the duration of the pattern change warning time.
                (Components|InstancedStaticMesh|SetCustomDataValue _hismcomponentoutlet_3 _instanceindex_4 (Variables|Laser|Material|GetOutletMaterialCustomDataIndex_EmissiveEnabled) 1.0 true)
                (Components|InstancedStaticMesh|SetCustomDataValue _hismcomponentoutlet_3 _instanceindex_4 (Variables|Laser|Material|GetOutletMaterialCustomDataIndex_DisableEmissivePulse) 0.0 true)
                (Components|InstancedStaticMesh|SetCustomDataValue _hismcomponentoutlet_3 _instanceindex_4 (Variables|Laser|Material|GetOutletMaterialCustomDataIndex_AnimTimestamp) (Utilities|Time|GetGameTimeinSeconds) true)
                (Components|InstancedStaticMesh|SetCustomDataValue _hismcomponentoutlet_3 _instanceindex_4 (Variables|Laser|Material|GetOutletMaterialCustomDataIndex_AnimTimestampDuration) (Variables|Laser|GetLaserPatternChangeWarningTime) true)
                (bind _soundoutlettransitionon (Variables|Laser|Sound|GetSoundOutletTransitionOn)
                  (:then
                    (Audio|PlaySoundatLocation _soundoutlettransitionon _outinstancetransform_location))
                  (:else)))
              (:On
                ;;[0.1302 0.0375 0.06926] Outlet is ON. Emissive is at the brightest.
                (Components|InstancedStaticMesh|SetCustomDataValue _hismcomponentoutlet_2 _instanceindex_2 (Variables|Laser|Material|GetOutletMaterialCustomDataIndex_EmissiveEnabled) 1.0 true)
                (Components|InstancedStaticMesh|SetCustomDataValue _hismcomponentoutlet_2 _instanceindex_2 (Variables|Laser|Material|GetOutletMaterialCustomDataIndex_AnimTimestamp) 0.0 true)
                (Components|InstancedStaticMesh|SetCustomDataValue _hismcomponentoutlet_2 _instanceindex_2 (Variables|Laser|Material|GetOutletMaterialCustomDataIndex_AnimTimestampDuration) 0.01 true))
              (:TransitionOff
                ;;[0.01714 0.04625 0.08854] Outlet is transitioning off.
                (Components|InstancedStaticMesh|SetCustomDataValue _hismcomponentoutlet _instanceindex_3 (Variables|Laser|Material|GetOutletMaterialCustomDataIndex_EmissiveEnabled) 1.0 true)
                (Components|InstancedStaticMesh|SetCustomDataValue _hismcomponentoutlet _instanceindex_3 (Variables|Laser|Material|GetOutletMaterialCustomDataIndex_AnimTimestamp) (Utilities|Operators|Add (Utilities|Time|GetGameTimeinSeconds) (* (Math|Float|SafeDivide (Variables|Laser|Material|GetMaterialCosineWavePeriod) 2.0) -1.0) (Math|Random|RandomFloatinRange _x _y)) true)))))
        (:else)))))

(fn DeprecatedUpdateLaserConnectionCustomData
  (params (InstanceStaticMeshIndex :type Integer) (LaserConnectionData :type Laser%20Grid%20Connection%20Entry%20Structure%20%28by%20ref%29))
  (bind (_primaryoutletindex _secondaryoutletindex) (Utilities|Struct|BreakLaserOutletConnection (Utilities|Struct|BreakLaserGridConnectionEntry LaserConnectionData)))
  (bind (_outletconnection _state) (Utilities|Struct|BreakLaserGridConnectionEntry LaserConnectionData))
  (bind _returnvalue (Math|Conversions|ToInteger(Byte) _primaryoutletindex))
  (bind _returnvalue_2 (Math|Conversions|ToInteger(Byte) _secondaryoutletindex))
  (bind _hismcomponentlaser (Variables|Laser|GetHISMComponentLaser))
  (bind (_lasermaterialrandomflickeringspeed_x _lasermaterialrandomflickeringspeed_y) (Math|Vector2D|BreakVector2D (Variables|GetLaserMaterialRandomFlickeringSpeed_X)))
  ;[0 0 0] UpdateLaserConnectionCustomData: Update the laser material settings and play the appropriate sounds based on the laser connection data.
  ; THIS FUNCTION HAS BEEN NATIVIZED AND IS NOW DEPRECATED. IT IS ONLY HERE TO HELP ARTISTS AND DEV COMMUNICATE REGARDING THE EXISTING LOGIC.
  (Networking|CanExecuteCosmeticEvents
    (:True
      (Variables|Laser|GetHISMComponentLaser
        (:then
          (switch Utilities|FlowControl|Switch|SwitchonELaserConnectionState _state
            (:On
              ;;[0.1298 0.03179 0.1302] Setup the material state for the laser to be ON.
              (Components|InstancedStaticMesh|SetCustomDataValue _hismcomponentlaser InstanceStaticMeshIndex (Variables|Laser|Material|GetLaserMaterialCustomDataIndex_RandomFlickeringSpeed) (Math|Random|RandomFloatinRange _lasermaterialrandomflickeringspeed_x _lasermaterialrandomflickeringspeed_y) true)
              (Components|InstancedStaticMesh|SetCustomDataValue _hismcomponentlaser InstanceStaticMeshIndex (Variables|Laser|Material|GetLaserMaterialCustomDataIndex_AnimTimestamp) (Utilities|Time|GetGameTimeinSeconds) true)
              (Components|InstancedStaticMesh|SetCustomDataValue _hismcomponentlaser InstanceStaticMeshIndex (Variables|Laser|Material|GetLaserMaterialCustomDataIndex_PulseEnabled) 0.0 true))
            (:TransitionOff
              ;;[0.01082 0.06771 0.02314] Setup the material state for the laser to be transitioning
              (Components|InstancedStaticMesh|SetCustomDataValue _hismcomponentlaser InstanceStaticMeshIndex (Variables|Laser|Material|GetLaserMaterialCustomDataIndex_PulseEnabled) 1.0 true)
              (Components|InstancedStaticMesh|SetCustomDataValue _hismcomponentlaser InstanceStaticMeshIndex (Variables|Laser|Material|GetLaserMaterialCustomDataIndex_PulseTimeOffset) (+ (Utilities|Time|GetGameTimeinSeconds) (Math|Float|SafeDivide (Variables|Laser|Material|GetMaterialCosineWavePeriod) 2.0)) true)))
          (bind _hismcomponentoutlet (Variables|Laser|GetHISMComponentOutlet)
            (:then
              ;;[0.01307 0.05501 0.05729] Play the laser state transition sound on the laser outlets as well as the laser beam.
              (if (and (.ReturnValue (Components|InstancedStaticMesh|GetInstanceTransform _hismcomponentoutlet _returnvalue true)) (.ReturnValue (Components|InstancedStaticMesh|GetInstanceTransform _hismcomponentoutlet _returnvalue_2 true)))
                (bind _soundlasertransitionprimaryoutlet (Variables|Laser|Sound|GetSoundLaserTransitionPrimaryOutlet)
                  (:then
                    (Audio|PlaySoundatLocation _soundlasertransitionprimaryoutlet (.location (Components|InstancedStaticMesh|GetInstanceTransform _hismcomponentoutlet _returnvalue true))))
                  (:else
                    _))
                (bind _soundlasertransitionsecondaryoutlet (Variables|Laser|Sound|GetSoundLaserTransitionSecondaryOutlet)
                  (:then
                    (Audio|PlaySoundatLocation _soundlasertransitionsecondaryoutlet (.location (Components|InstancedStaticMesh|GetInstanceTransform _hismcomponentoutlet _returnvalue_2 true))))
                  (:else
                    _))
                (else
                  _))
              (if (Utilities|Enum|Equal(Enum) _state "TransitionOff")
                (Audio|PlaySoundatLocation (Variables|Laser|Sound|GetSoundLaserTransition) (/ (+ (.location (Components|InstancedStaticMesh|GetInstanceTransform _hismcomponentoutlet _returnvalue true)) (.location (Components|InstancedStaticMesh|GetInstanceTransform _hismcomponentoutlet _returnvalue_2 true))) 2))))
            (:else)))
        (:else)))))

(fn DisableLaser
  (bind _self self)
  (bind _returnvalue (Ability|GetFortAbilitySystemComponent _self))
  ;[1 0.9969 0.2764] Disable the laser grid by applying the disable gameplay effect.
  (Networking|SwitchHasAuthority
    (:Authority
      (Utilities|IsValid _returnvalue
        (:Is%20Valid
          (GameplayEffects|ApplyGameplayEffectToSelf _returnvalue (Variables|Settings|GetDisableLaserGameplayEffectClass) 0.0 (GameplayEffects|MakeEffectContext _returnvalue)))))))

(fn ExecuteLaserGridTriggeredFX
  (bind _vfxtransforms (LocalVariables|GetVFXTransforms))
  ;[0.1739 0.03384 0.224] ExecuteLaserGridTriggeredFX: Calculate the transforms of the various triggered FXs to spawn on the client based on the active connections and multicast them to the client.
  ; We may wanna move the transform calculating code to CPP cause a similar calculation is being done there for the laser component.
  (Networking|SwitchHasAuthority
    (:Authority
      (bind _outactivelasergridconnections (Laser|GetActiveLaserGridConnections))
      (for _array_element _outactivelasergridconnections
        (bind _hismcomponentoutlet (Variables|Laser|GetHISMComponentOutlet)
          (:then
            (bind (_outletconnection_primaryoutletindex _outletconnection_secondaryoutletindex) (Utilities|Struct|BreakLaserOutletConnection (Utilities|Struct|BreakLaserGridConnectionEntry _array_element)))
            (if (and (Components|InstancedStaticMesh|IsValidInstance _hismcomponentoutlet (Math|Conversions|ToInteger(Byte) _outletconnection_primaryoutletindex)) (Components|InstancedStaticMesh|IsValidInstance _hismcomponentoutlet (Math|Conversions|ToInteger(Byte) _outletconnection_secondaryoutletindex)))
              (bind _outinstancetransform (Components|InstancedStaticMesh|GetInstanceTransform _hismcomponentoutlet (Math|Conversions|ToInteger(Byte) _outletconnection_secondaryoutletindex) true))
              (bind _outinstancetransform_location (Math|Transform|BreakTransform _outinstancetransform))
              (bind _outinstancetransform_2 (Components|InstancedStaticMesh|GetInstanceTransform _hismcomponentoutlet (Math|Conversions|ToInteger(Byte) _outletconnection_primaryoutletindex) true))
              (bind _outinstancetransform_location_2 (Math|Transform|BreakTransform _outinstancetransform_2))
              (Utilities|Array|Add (LocalVariables|GetVFXTransforms) (Math|Transform|MakeTransform _outinstancetransform_location_2 (Math|Conversions|RotationFromXVector (Math|Vector|GetUnitDirection(Vector) _outinstancetransform_location_2 _outinstancetransform_location)) (Math|Vector|MakeVector (Math|Float|SafeDivide (Math|Vector|Distance(Vector) _outinstancetransform_location_2 _outinstancetransform_location) (Variables|Settings|GetMaxScaleBeamDistance)) 1.0 1.0)))))
          (:else)))
      (if (Utilities|Array|IsNotEmpty _vfxtransforms)
        (CallFunction|MulticastLaserGridTriggeredFX _vfxtransforms)))))

(fn GetPlayerPawnsFromTriggeringActor
  (params (Actor :type Actor%20Object%20Reference))
  ;[0.02947 0.148 0.1927] Return all the player pawns from the input actor.
  ; This function handles players that are riding wildlife and players that are in the vehicle as well.
  (Utilities|FlowControl|Sequence
    (:then_0
      (bind _asfort_player_pawn (Utilities|Casting|CastToFortPlayerPawn Actor)
        (:then
          (Utilities|Array|AddUnique (LocalVariables|GetRetPlayerPawns) _asfort_player_pawn))
        (:CastFailed
          (bind _asfort_aipawn (Utilities|Casting|CastToFortAIPawn Actor))
          (bind _attachedplayerpawns (Default|GetPlayersAttachedtoAIPawn _asfort_aipawn))
          (Utilities|Array|AppendArray (LocalVariables|GetRetPlayerPawns) _attachedplayerpawns))))
    (:then_1
      (return (LocalVariables|GetRetPlayerPawns)))))

(fn GetPlayersAttachedToAIPawn
  (params (AIPawn :type Fort%20AIPawn%20Object%20Reference))
  ;[0.00971 0.1011 0.1042] Returns the attached players to an AI Pawn.
  (Utilities|FlowControl|Sequence
    (:then_0
      (Utilities|IsValid AIPawn
        (:Is%20Valid
          (for _array_element (Actor|GetAttachedActors AIPawn true false)
            (bind _asfort_player_pawn (Utilities|Casting|CastToFortPlayerPawn _array_element))
            (Utilities|Array|Add (LocalVariables|GetRetPlayerPawns) _asfort_player_pawn)))))
    (:then_1
      (return (LocalVariables|GetRetPlayerPawns)))))

(fn MarkPlayerPawn
  (params (PlayerPawn :type Fort%20Player%20Pawn%20Object%20Reference))
  (bind (_returnvalue_x _returnvalue_y _returnvalue_z) (Math|Vector|BreakVector (Transformation|GetActorLocation)))
  (bind _returnvalue (Ability|GetAbilitySystemComponent PlayerPawn))
  ;[0.01358 0.1354 0.1246] Mark the input player pawn and alert nearby guards.
  (Networking|SwitchHasAuthority
    (:Authority
      (Utilities|IsValid PlayerPawn
        (:Is%20Valid
          (Utilities|IsValid _returnvalue
            (:Is%20Valid
              (GameplayEffects|ApplyGameplayEffectToSelf _returnvalue (Variables|Settings|GetLaserMarkPlayerGameplayEffect) 0.0 (GameplayEffects|MakeEffectContext _returnvalue))
              (if (ScalableFloat|GetValueasBool (Variables|Hotfix|GetHotfix_PlayerMarkGuardsAlertEnabled))
                (bind _outhits (Collision|MultiSphereTraceByChannel (Transformation|GetActorLocation) (+ (Transformation|GetActorLocation) "0.000000,0.000000,1.000000") (ScalableFloat|GetValueatLevel (Variables|Hotfix|GetHotfix_PlayerMarkGuardsAlertBroadcastRadius)) "TraceTypeQuery9"))
                (for _array_element _outhits
                  (bind _asfort_player_pawn (Utilities|Casting|CastToFortPlayerPawn (.HitActor (Collision|BreakHitResult _array_element))))
                  (if (GameplayTags|HasTag (Fortnite|Factions|GetFactionTagContainerforActor _asfort_player_pawn) (Variables|Settings|GetAlertGuardsFactionTag) true)
                    (bind (_returnvalue_x_2 _returnvalue_y_2 _returnvalue_z_2) (Math|Vector|BreakVector (Transformation|GetActorLocation _asfort_player_pawn)))
                    (if (<= (Math|Float|Absolute(Float) (- _returnvalue_z_2 _returnvalue_z)) (ScalableFloat|GetValueatLevel (Variables|Hotfix|GetHotfix_PlayerMarkGuardsAlertBroadcastHeight)))
                      (bind _asfort_athena_aibot_controller (Utilities|Casting|CastToFortAthenaAIBotController (Pawn|GetController _asfort_player_pawn)))
                      (Class|FortAthenaAIBotController|QueueStim _asfort_athena_aibot_controller PlayerPawn (Transformation|GetActorLocation PlayerPawn) (Math|Vector|GetUnitDirection(Vector) (Transformation|GetActorLocation _asfort_player_pawn) (Transformation|GetActorLocation PlayerPawn)) "Enemy" 1.0 "None"))))))))))))

(fn RemoveInstance
  ;[0.03296 0.02431 0.07812] Remove the last instance from the outlet mesh component.
  (if (Editor|InEditor)
    (bind _hismcomponentoutlet (Variables|Laser|GetHISMComponentOutlet)
      (:then
        (Components|InstancedStaticMesh|RemoveInstance _hismcomponentoutlet (- (Components|InstancedStaticMesh|GetInstanceCount _hismcomponentoutlet) 1)))
      (:else))))

(fn ShouldPlayTriggeredCue
  (params (FortPlayerPawn :type Fort%20Player%20Pawn%20Object%20Reference))
  (Utilities|IsValid FortPlayerPawn
    (:Is%20Valid
      (if (ScalableFloat|GetValueasBool (Variables|Hotfix|GetHotfix_PlayTriggeredCueOnAlreadyMarkedPawn))
        (return true)
        (else
          (return (not (GameplayTags|HasMatchingGameplayTag FortPlayerPawn (Variables|Laser|GetMarkedByLaserTag)))))))
    (:Is%20Not%20Valid
      (return false))))

(fn SnapInstance
  ;[0.06306 0.1749 0.2083] SnapInstance: Snaps all the outlet instances to the nearest mesh that can be found within 100 units.
  (if (Editor|InEditor)
    (bind _hismcomponentoutlet (Variables|Laser|GetHISMComponentOutlet)
      (:then
        (for _index (range (Components|InstancedStaticMesh|GetInstanceCount _hismcomponentoutlet))
          (bind (_outinstancetransform _returnvalue) (Components|InstancedStaticMesh|GetInstanceTransform _hismcomponentoutlet _index true))
          (bind (_outinstancetransform_location _outinstancetransform_rotation _outinstancetransform_scale) (Math|Transform|BreakTransform _outinstancetransform))
          (if _returnvalue
            (bind (_outhit _returnvalue_2) (Collision|LineTraceByChannel (+ _outinstancetransform_location (* (Math|Vector|GetRightVector _outinstancetransform_rotation) 1e+02)) (- _outinstancetransform_location (* (Math|Vector|GetRightVector _outinstancetransform_rotation) 1e+02)) "TraceTypeQuery10" true))
            (if _returnvalue_2
              (Components|InstancedStaticMesh|UpdateInstanceTransform _hismcomponentoutlet _index (Math|Transform|MakeTransform (.ImpactPoint (Collision|BreakHitResult _outhit)) _outinstancetransform_rotation _outinstancetransform_scale) true true false)))))
      (:else))))

(fn UpdateLaserConnectionCosmetics
  (params (Array :type Array%20of%20Laser%20Grid%20Connection%20Entry%20Structures))
  (bind _audio (Variables|Default|GetAudio))
  (bind _areanylasersactive (LocalVariables|GetAreAnyLasersActive))
  ;[0.9598 0.974 0.1975 0.5] Figure out if any lasers are active.
  (bind _array_element (Utilities|Array|ForEachLoopwithBreak Array)
    (:LoopBody
      (if (Utilities|Enum|NotEqual(Enum) (.State (Utilities|Struct|BreakLaserGridConnectionEntry _array_element)) "Off")
        (LocalVariables|SetAreAnyLasersActive true)))
    (:Completed
      _))
  ;[0.1656 1 0.9602 0.5] Turn on/off the audio loop based on if any lasers are currently active.
  (if (!= (Components|Activation|IsActive _audio) _areanylasersactive)
    (Components|Activation|SetActive _audio _areanylasersactive false)))

(fn ConstructionScript)
