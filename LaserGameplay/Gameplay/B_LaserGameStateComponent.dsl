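; Server-side game state component that tracks laser grid groups and reports which were triggered as analytics.
; On BeginPlay it listens for game phase changes and finds groups through the global actor registry by gameplay tag, including later registrations.
; It records each group's name on its first trigger, unbinding afterward.
; At EndGame, or EndPlay in editor, it fires one Core.LaserGameplay event, skipped during replays.
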(event EventBeginPlay
  (bind _returnvalue (WorldSubsystems|GetFortGlobalActorRegistrySubsystem))
  (bind _lasergroupactorgameplaytag (Variables|Settings|GetLaserGroupActorGameplayTag))
  ;[0.02051 0.09375 0.02889] BeginPlay: Bind to know when the game phase changes.
  ; Set up delegates to find and register all the laser grid group actors in the world.
  (|CallParentFunction)
  (Utilities|SwitchHasAuthority
    (:Authority
      ;;[0.07311 0.2232 0.3125] Bind to know when the game phase changes.
      (bind _gamestate (Class|FortAsyncActionCurrentPlaylistReady|PlaylistDataReadyAsync)
        (:OnReady
          (bind _returnvalue_2 (GameplayEventRouter|GetEventRouter "/Script/FortniteGame.FortGameState" _gamestate))
          (bind (_asynctaskproxy _payload) (GameplayEventRouter|StartListeningToStatefulEvent _returnvalue_2 "/Script/FortniteGame.FortGamePhaseUpdatedEvent")
            (:then
              (Variables|Default|SetGamePhaseUpdatedAsyncTask _asynctaskproxy)
              ;;[0.1612 0.04554 0.1667] Set up delegates to find and register all the laser grid group actors in the world.
              (Utilities|IsValid _returnvalue
                (:Is%20Valid
                  (bind _outactors (Class|FortGlobalActorRegistrySubsystem|GetRegisteredActors _returnvalue _lasergroupactorgameplaytag))
                  (for _array_element _outactors
                    (Utilities|IsValid _array_element
                      (:Is%20Valid
                        (CallFunction|RegisterLaserGridGroup :RegisteredActor _array_element))))
                  (bind _returnvalue_3 (Class|FortGlobalActorRegistrySubsystem|GetActorRegistryDataforGameplayTag _returnvalue _lasergroupactorgameplaytag))
                  (Utilities|IsValid _returnvalue_3
                    (:Is%20Valid
                      (EventDispatchers|BindEventtoOnActorRegistered _returnvalue_3 (EventDispatchers|CreateEvent :!event RegisterLaserGridGroup)))))))
            (:OnEventReceived
              _)
            (:OnSavedEventStateExists
              _))
          (Default|HandleGamePhaseChanged (Utilities|Struct|BreakFortGamePhaseUpdatedEvent _payload)))))))

(event EventEndPlay
  (params (EndPlayReason :type EEndPlayReason%20Enum :default "Destroyed"))
  (bind _returnvalue (WorldSubsystems|GetFortGlobalActorRegistrySubsystem))
  ;[0.1042 0.07147 0.008582] EndPlay: Unbind delegates and submit analytics.
  (Utilities|FlowControl|Sequence
    (:then_0
      (Utilities|SwitchHasAuthority
        (:Authority
          (Utilities|FlowControl|Sequence
            (:then_0
              (if (Editor|InEditororPIE)
                (Default|SubmitAnalytics)
                (else
                  _))
              (bind _gamephaseupdatedasynctask (Variables|Default|GetGamePhaseUpdatedAsyncTask)
                (:then
                  (GameplayEventRouter|StopListeningtoStatefulEvent _gamephaseupdatedasynctask))
                (:else)))
            (:then_1
              (Utilities|IsValid _returnvalue
                (:Is%20Valid
                  (bind _returnvalue_2 (Class|FortGlobalActorRegistrySubsystem|GetActorRegistryDataforGameplayTag _returnvalue (Variables|Settings|GetLaserGroupActorGameplayTag)))
                  (Utilities|IsValid _returnvalue_2
                    (:Is%20Valid
                      (EventDispatchers|UnbindEventfromOnActorRegistered _returnvalue_2 (EventDispatchers|CreateEvent :!event RegisterLaserGridGroup)))))))
            (:then_2
              (for _array_element (Variables|Transient|GetRegisteredLaserGridGroups)
                (Utilities|IsValid _array_element
                  (:Is%20Valid
                    (Default|UnbindEventfromOnLaserGridGroupTriggered _array_element (EventDispatchers|CreateEvent :!event HandleLaserGridGroupTriggered))))))))))
    (:then_1
      (|CallParentFunction EndPlayReason))))

(fn HandleGamePhaseChanged
  (params (GamePhase :type EAthenaGamePhase%20Enum))
  ;[0.0164 0.07935 0.1198] HandleGamePhaseChanged: Once we're in EndGame, submit the analytics.
  (Networking|SwitchHasAuthority
    (:Authority
      (if (Utilities|Enum|Equal(Enum) GamePhase "EndGame")
        (Default|SubmitAnalytics)
        (bind _gamephaseupdatedasynctask (Variables|Default|GetGamePhaseUpdatedAsyncTask)
          (:then
            (GameplayEventRouter|StopListeningtoStatefulEvent _gamephaseupdatedasynctask))
          (:else))))))

(fn HandleLaserGridGroupTriggered
  (params (LaserGridGroup :type B%20Laser%20Grid%20Group%20Object%20Reference))
  ;[0.01227 0.07492 0.09375] Track the names of the laser grid group that has been triggered. This will be passed through analytics.
  (Utilities|SwitchHasAuthority
    (:Authority
      (Utilities|IsValid LaserGridGroup
        (:Is%20Valid
          (Default|UnbindEventfromOnLaserGridGroupTriggered LaserGridGroup (EventDispatchers|CreateEvent :!event HandleLaserGridGroupTriggered))
          (Utilities|Array|AddUnique (Variables|Transient|Analytics|GetAnalytics_LaserGridGroupsTriggered) (Utilities|GetDisplayName LaserGridGroup)))))))

(fn RegisterLaserGridGroup
  (params (GameplayTag :type Gameplay%20Tag%20Structure%20%28by%20ref%29) (RegisteredActor :type Actor%20Object%20Reference))
  ;[0.04421 0.02472 0.1146] RegisterLaserGridGroup: Register the laser grid group.
  (Utilities|FlowControl|Sequence
    (:then_0
      (Utilities|SwitchHasAuthority
        (:Authority
          (bind _asb_laser_grid_group (Utilities|Casting|CastToB_LaserGridGroup RegisteredActor))
          (bind _returnvalue (Utilities|Array|AddUnique (Variables|Transient|GetRegisteredLaserGridGroups) _asb_laser_grid_group))
          (if (>= _returnvalue 0)
            (Default|BindEventtoOnLaserGridGroupTriggered _asb_laser_grid_group (EventDispatchers|CreateEvent :!event HandleLaserGridGroupTriggered))
            (Fortnite|FortShippingLog self (Utilities|String|Append "RegisterLaserGridGroup: Registered laser group: " (Utilities|GetDisplayName _asb_laser_grid_group)))))))
    (:then_1
      (return))))

(fn SubmitAnalytics
  (bind _analytics_lasergridgroupstriggered (Variables|Transient|Analytics|GetAnalytics_LaserGridGroupsTriggered))
  (bind _returnvalue (Utilities|Array|Length _analytics_lasergridgroupstriggered))
  (bind _registeredlasergridgroups (Variables|Transient|GetRegisteredLaserGridGroups))
  (bind _returnvalue_2 (Utilities|Array|Length _registeredlasergridgroups))
  (bind _analytics_lasergridgroupstriggered_2 (Variables|Transient|Analytics|GetAnalytics_LaserGridGroupsTriggered))
  (bind _returnvalue_3 (Utilities|String|JoinStringArray _analytics_lasergridgroupstriggered_2 " ; "))
  ;[0.01098 0.08502 0.08854] SubmitAnalytics: Fire analytics if it hasn't been done yet.
  (Utilities|SwitchHasAuthority
    (:Authority
      (if (and (not (Replay|GetIsPlayingReplay)) (not (Variables|Transient|Analytics|GetbAnalyticsSubmitted)))
        (Variables|Transient|Analytics|SetbAnalyticsSubmitted true)
        (Utilities|FlowControl|Sequence
          (:then_0
            ;;[0.02283 0.151 0.08822] NumOfLaserGroupsTriggered
            (bind _returnvalue_4 (Analytics|ConstructAnalyticsAttribute :AttributeName "NumOfLaserGroupsTriggered" :IntValue _returnvalue :floatValue -1.0 :BoolValue false))
            (Utilities|Array|Add (LocalVariables|GetAnalyticsEventAttributes) _returnvalue_4)
            (Fortnite|FortShippingLog self (Utilities|String|Append "NumOfLaserGroupsTriggered: " (Utilities|String|ToString(Integer) _returnvalue)))
            ;;[0.02424 0.08952 0.1823] LaserGroupsTriggered
            (bind _returnvalue_5 (Analytics|ConstructAnalyticsAttribute "LaserGroupsTriggered" _returnvalue_3))
            (Utilities|Array|Add (LocalVariables|GetAnalyticsEventAttributes) _returnvalue_5)
            (Fortnite|FortShippingLog self (Utilities|String|Append "LaserGroupsTriggered: " _returnvalue_3))
            ;;[0.1823 0.03399 0.1581] TotalNumOfLaserGroups
            (bind _returnvalue_6 (Analytics|ConstructAnalyticsAttribute :AttributeName "TotalNumOfLaserGroupsInWorld" :IntValue _returnvalue_2 :floatValue -1.0 :BoolValue false))
            (Utilities|Array|Add (LocalVariables|GetAnalyticsEventAttributes) _returnvalue_6)
            (Fortnite|FortShippingLog self (Utilities|String|Append "TotalNumOfLaserGroups: " (Utilities|String|ToString(Integer) _returnvalue_2))))
          (:then_1
            (Analytics|FireAnalyticsEventBlueprintArray "Core.LaserGameplay" (LocalVariables|GetAnalyticsEventAttributes))))))))
