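; B_LaserGridGroup links a set of laser grids so that tripping one triggers the whole group.
; Once playlist data is ready, the server binds to each grid's triggered and outlet-destroyed events, treating a destroyed outlet as a trigger.
; A trigger disables every laser, unbinds the delegates and broadcasts OnLaserGridGroupTriggered for listeners.
; EndPlay also unbinds, and all of this logic runs server-side only.
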(event EventBeginPlay
  ;[0.01955 0.1042 0.03483] BeginPlay: Bind to know when the laser grid has been triggered.
  (|CallParentFunction)
  (Default|SwitchNetMode
    (:Server
      (Class|FortAsyncActionCurrentPlaylistReady|PlaylistDataReadyAsync
        (:OnReady
          (for _array_element (Variables|Default|GetLaserGrids)
            (Utilities|IsValid _array_element
              (:Is%20Valid
                (Laser|BindEventtoOnLaserGridTriggered _array_element (EventDispatchers|CreateEvent :!event Custom|HandleLaserGridTriggered))
                (Laser|BindEventtoOnLaserGridOutletDestroyed _array_element (EventDispatchers|CreateEvent :!event HandleLaserGridOutletDestroyed))
                (Fortnite|FortShippingLog self (Utilities|String|Append "BeginPlay: Binds setup for laser grid : " (Utilities|GetDisplayName _array_element)))))))))))

(event EventEndPlay
  (params (EndPlayReason :type EEndPlayReason%20Enum :default "Destroyed"))
  ;[0.08854 0.05824 0.01999] EndPlay: Unbind delegates.
  (Utilities|FlowControl|Sequence
    (:then_0
      (Default|SwitchNetMode
        (:Server
          (Default|UnbindLaserGridDelegates))))
    (:then_1
      (|CallParentFunction EndPlayReason))))

(event Custom|HandleLaserGridTriggered
  (params (LaserGrid :type Building%20Gameplay%20Actor%20Laser%20Grid%20Object%20Reference) (LaserConnectionData :type Laser%20Grid%20Connection%20Entry%20Structure%20%28by%20ref%29) (TriggeringActor :type Actor%20Object%20Reference))
  ;[0.0451 0.1994 0.2917] HandleLaserGridTriggered: Disable all the lasers, unbind delegates, and close all the linked security gates.
  (Default|SwitchNetMode
    (:Server
      (Fortnite|FortShippingLog self (Utilities|String|Append "HandleLaserGridTriggered: Triggered by LaserGrid: " (Utilities|GetDisplayName LaserGrid)))
      (Default|DisableLasers)
      (Default|UnbindLaserGridDelegates)
      (Default|CallOnLaserGridGroupTriggered self))))

(fn DisableLasers
  ;[0.1458 0.03971 0.08151] Disable all the lasers.
  (Default|SwitchNetMode
    (:Server
      (Fortnite|FortShippingLog self "DisableLasers: Invoked.")
      (for _array_element (Variables|Default|GetLaserGrids)
        (Utilities|IsValid _array_element
          (:Is%20Valid
            (Laser|DisableLaser _array_element)))))))

(fn HandleLaserGridOutletDestroyed
  (params (LaserGrid :type Building%20Gameplay%20Actor%20Laser%20Grid%20Object%20Reference))
  ;[0.02704 0.1719 0.1513] HandleLaserGridOutletDestroyed: When an outlet is destroyed, treat it like the laser grid has been triggered.
  (Utilities|FlowControl|Sequence
    (:then_0
      (Default|SwitchNetMode
        (:Server
          (Utilities|IsValid LaserGrid
            (:Is%20Valid
              (Fortnite|FortShippingLog self (Utilities|String|Append "HandleLaserGridOutletDestroyed: " (Utilities|GetDisplayName LaserGrid)))
              (CallFunction|HandleLaserGridTriggered LaserGrid))))))
    (:then_1
      (return))))

(fn UnbindLaserGridDelegates
  ;[0.01326 0.1534 0.1667] Unbind delegates.
  (Default|SwitchNetMode
    (:Server
      (for _array_element (Variables|Default|GetLaserGrids)
        (Utilities|IsValid _array_element
          (:Is%20Valid
            (Laser|UnbindEventfromOnLaserGridTriggered _array_element (EventDispatchers|CreateEvent :!event HandleLaserGridTriggered))
            (Laser|UnbindEventfromOnLaserGridOutletDestroyed _array_element (EventDispatchers|CreateEvent :!event HandleLaserGridOutletDestroyed))))))))

(fn ConstructionScript)
