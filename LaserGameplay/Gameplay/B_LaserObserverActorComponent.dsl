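; Actor component that watches a set of laser grid actors and raises a single alarm event when any is tripped.
; On BeginPlay it binds to each valid grid's triggered and outlet-destroyed events, with all logic running only on the server.
; A touch or destroyed outlet disables every laser, unbinds all delegates and broadcasts OnLaserTriggered, making the response one-shot.
; EndPlay also removes the bindings.
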(event EventBeginPlay
  ;[0.04924 0.1198 0.02446] BeginPlay: Bind to know when the lasers have touched an actor.
  (|CallParentFunction)
  (Utilities|SwitchNetRole
    (:Server
      (for _array_element (Variables|Default|GetLaserGridActors)
        (Utilities|IsValid _array_element
          (:Is%20Valid
            (Laser|BindEventtoOnLaserGridTriggered _array_element (EventDispatchers|CreateEvent :!event HandleLaserGridTriggered))
            (Laser|BindEventtoOnLaserGridOutletDestroyed _array_element (EventDispatchers|CreateEvent :!event HandleLaserGridOutletDestroyed))))))))

(event EventEndPlay
  (params (EndPlayReason :type EEndPlayReason%20Enum :default "Destroyed"))
  ;[0.1667 0.1029 0.02029] EndPlay: Unbind delegates.
  (|CallParentFunction EndPlayReason)
  (Utilities|SwitchNetRole
    (:Server
      (CallFunction|UnbindDelegates))))

(fn DisableLasers
  ;[0.1458 0.03971 0.08151] SetLaserActorsEnable: Activate or deactivate all the lasers.
  (Utilities|SwitchNetRole
    (:Server
      (for _array_element (Variables|Default|GetLaserGridActors)
        (Utilities|IsValid _array_element
          (:Is%20Valid
            (Laser|DisableLaser _array_element)))))))

(fn HandleLaserGridOutletDestroyed
  (params (LaserGrid :type Building%20Gameplay%20Actor%20Laser%20Grid%20Object%20Reference))
  ;[0.01818 0.1523 0.1771] HandleLaserGridOutletDestroyed: Broadcast that the laser grid has been triggered.
  (Utilities|SwitchNetRole
    (:Server
      (CallFunction|HandleLaserGridTriggered 0))))

(fn HandleLaserGridTriggered
  (params (LaserGrid :type Building%20Gameplay%20Actor%20Laser%20Grid%20Object%20Reference) (LaserConnectionData :type Laser%20Grid%20Connection%20Entry%20Structure) (TriggeringActor :type Actor%20Object%20Reference))
  ;[0.1438 0.05357 0.1667] HandleLaserGridTriggered: Handle one of the lasers being touched.
  (Utilities|SwitchNetRole
    (:Server
      (CallFunction|DisableLasers)
      (CallFunction|UnbindDelegates)
      (Default|CallOnLaserTriggered))))

(fn UnbindDelegates
  ;[0.01326 0.1534 0.1667] Unbind delegates.
  (Utilities|SwitchNetRole
    (:Server
      (for _array_element (Variables|Default|GetLaserGridActors)
        (Utilities|IsValid _array_element
          (:Is%20Valid
            (Laser|UnbindEventfromOnLaserGridTriggered _array_element (EventDispatchers|CreateEvent :!event HandleLaserGridTriggered))
            (Laser|UnbindEventfromOnLaserGridOutletDestroyed _array_element (EventDispatchers|CreateEvent :!event HandleLaserGridOutletDestroyed))))))))
